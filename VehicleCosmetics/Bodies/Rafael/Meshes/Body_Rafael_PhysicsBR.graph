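# Physics asset for the Rafael vehicle body cosmetic mesh, likely the BR variant going by its name.
# It defines a single physics body on the root bone and has no constraints, so the body has no articulated joints.
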PhysicsAsset /VehicleCosmetics/Bodies/Rafael/Meshes/Body_Rafael_PhysicsBR
bodies {root []}
constraints []
