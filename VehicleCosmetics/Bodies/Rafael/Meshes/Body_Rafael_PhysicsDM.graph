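# Body_Rafael_PhysicsDM is a physics asset for the Rafael vehicle body mesh in the cosmetics set.
# It holds a single body on the root bone, built from one box collision shape offset upward and slightly forward from the bone.
# It has no constraints, so the whole body acts as one simple rigid collision volume.
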PhysicsAsset /VehicleCosmetics/Bodies/Rafael/Meshes/Body_Rafael_PhysicsDM
bodies {root [{shapeType=Box, extentX=444, extentY=216, extentZ=116, shapeName=root_box, center=<43.1402 0 93>}]}
constraints []
